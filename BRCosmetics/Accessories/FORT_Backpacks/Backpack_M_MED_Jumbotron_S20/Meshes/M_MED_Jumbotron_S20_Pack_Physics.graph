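# Physics asset for the Jumbotron S20 backpack cosmetic, setting up jointed bodies for its dangling zipper, ring and headphones.
# A capsule stands in for the pack, and a small box covers each of the three dyn_ joints.
# The zipper and ring are constrained directly to the pack, while the headphones chain off the ring with wider swing limits.
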PhysicsAsset /BRCosmetics/Accessories/FORT_Backpacks/Backpack_M_MED_Jumbotron_S20/Meshes/M_MED_Jumbotron_S20_Pack_Physics
bodies {
 dyn_headphones_joint [{shapeType=Box, extentX=5.5313, extentY=5.1622, extentZ=1.2997, shapeName=dyn_headphones_joint_box, center=<2.5563 0 -0.0797>}]
 dyn_ring_joint [{shapeType=Box, extentX=2.647, extentY=0.5242, extentZ=2.926, shapeName=dyn_ring_joint_box, center=<1.1473 0 0>}]
 dyn_zipper_joint [{shapeType=Box, extentX=1.8034, extentY=0.5, extentZ=0.7834, shapeName=dyn_zipper_joint_box, center=<0.034 0 0>}]
 pack [{shapeType=Capsule, length=10.4469, radius=9.5982, center=<-0.3055 -3.6171 -2.797>, rotation=<rot -5.8598 -90 0.6815>}]
}
constraints [
 {bone1Name=dyn_zipper_joint, bone2Name=pack, swing1LimitDegrees=75, swing2LimitDegrees=75, twistLimitDegrees=45}
 {bone1Name=dyn_ring_joint, bone2Name=pack, swing1LimitDegrees=45, swing2LimitDegrees=45, twistLimitDegrees=15}
 {bone1Name=dyn_headphones_joint, bone2Name=dyn_ring_joint, swing1LimitDegrees=95, swing2LimitDegrees=95, twistLimitDegrees=30}
]
