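# Animation Blueprint for the Jumbotron S20 backpack cosmetic, adding physics-driven secondary motion to the pack.
# The AnimGraph runs a RigidBody simulation in the pack bone's space on the input pose, then applies a small additive offset and roll to the root bone.
# The EventGraph updates gravity each frame through a deprecated override function and feeds it to the simulation.
# Falling begin and end events, each after a short retriggerable delay, set or clear an external force.
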AnimBlueprint /BRCosmetics/Accessories/FORT_Backpacks/Backpack_M_MED_Jumbotron_S20/Meshes/M_MED_Jumbotron_S20_Pack_AnimBP
target_skeleton=&/M_MED_Jumbotron_S20_Pack_Skeleton
graphs [
 AnimGraph {nodes [
   Local Space Ref Pose()
   n1 = RigidBody!(
    ComponentPose=Local To Component(LocalPose=Input Pose())
    Alpha
    OverrideWorldGravity:Vector=Get gravity() -> (gravity:Vector)
    ExternalForce=Get falling() -> (falling:Vector)
   ) -> (Pose) {
    bOverrideWorldGravity=true
    overridePhysicsAsset=&/M_MED_Jumbotron_S20_Pack_Physics
    simulationSpace=BaseBoneSpace
    alphaBoolBlend {blendOption=Linear}
    baseBoneRef {boneName=pack}
    componentLinearAccScale=<0.1 0.1 0.1>
    overrideWorldGravity {z=-980}
   }
   n2 = "Transform (Modify) Bone - Bone: root"(ComponentPose=@n1, Translation="0,1.5,0", Rotation="0,0,4") {
    rotationMode=BMM_Additive
    translationMode=BMM_Additive
    alphaBoolBlend {blendOption=Linear}
    boneToModify {boneName=root}
    rotation {roll=4}
    translation {y=1.5}
   }
   Output Pose(Result=Component To Local(ComponentPose=@n2)) {layerGroup=None}
  ]}
 EventGraph {nodes [
   Event BlueprintUpdateAnimation()
   > n1 = Gravity Override State Function DEPRECATED(CharacterPartAnimInstance=Self-Reference(), GameTimeInSeconds=GetGameTimeInSeconds())
   > Set gravity(gravity:Vector=@n1) -> (Output_Get:Vector)
   Event OnBeginFalling()
   > RetriggerableDelay(Duration=0.3)
   > Set falling(falling:Vector="30, 0,30") -> (Output_Get:Vector)
   Event OnEndFalling()
   > RetriggerableDelay(Duration=0.3)
   > Set falling(falling="0, 0, 0")
   TryGetPawnOwner()
  ]}
]
variables [
 falling {type=Vector}
 gravity {type=Vector}
]
